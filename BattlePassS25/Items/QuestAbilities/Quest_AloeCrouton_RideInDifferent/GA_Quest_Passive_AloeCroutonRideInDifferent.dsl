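; Passive Battle Pass quest ability (AloeCrouton) that tracks the player riding in different vehicles.
; On activation it runs two tag-wait tasks on the owning pawn, for InVehicle and AthenaInVehicle, both feeding one check.
; That check gets the pawn's current vehicle actor and, if absent from Used Vehicle List, stores it and broadcasts the RideInDifferentVehicles custom verb.
; Uniqueness is per vehicle actor instance, not vehicle type.
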(event Ability|EventFailedToActivatePassiveAbility
  (params (ActorInfo :type Gameplay%20Ability%20Actor%20Info%20Structure)))

(event Ability|EventActivateAbility
  (bind _owningplayerpawn (Variables|GAVars|Data|GetOwningPlayerPawn))
  (|CallParentFunction)
  (Ability|Tasks|WaitGameplayTagAddToActor _owningplayerpawn "(TagName=\"Gameplay.Action.Player.InVehicle\")"
    (:then
      (Ability|Tasks|WaitGameplayTagAddToActor _owningplayerpawn "(TagName=\"Gameplay.Action.Player.AthenaInVehicle\")"
        (:Added
          _)))
    (:Added
      _))
  (bind _returnvalue (Vehicle|GetVehicleActor (Vehicle|GetVehicleInterface _owningplayerpawn) :!class /Script/FortniteGame.FortVehicleInterface))
  (if (not (Utilities|Array|ContainsItem (Variables|Default|GetUsed%20Vehicle%20List) _returnvalue))
    (Utilities|Array|Add (Variables|Default|GetUsed%20Vehicle%20List) _returnvalue)
    (CustomVerb|BroadcastCustomVerbMessage (CustomVerb|MakeCustomVerbMessage (Variables|GAVars|Data|GetOwningPlayerPawn) "(TagName=\"CustomVerb.RideInDifferentVehicles\")"))))
